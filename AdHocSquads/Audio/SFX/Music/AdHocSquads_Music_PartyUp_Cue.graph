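# SoundCue that plays the party-up music stinger for the AdHocSquads feature.
# A single wave player feeds the AdHocSquads_Join_Party sound wave straight to the output.
# It routes through the Fort_Events_Music sound class at slightly reduced volume, with a very large max distance.
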SoundCue /AdHocSquads/Audio/SFX/Music/AdHocSquads_Music_PartyUp_Cue
root=WavePlayer_0
nodes [{id=WavePlayer_0, properties {soundWaveAssetPtr=&/AdHocSquads_Join_Party}}]
properties {maxDistance=2097152, soundClassObject=/Game/Sounds/FortSoundClassesAndModes/Fort_Events_Music, volumeMultiplier=0.8}
