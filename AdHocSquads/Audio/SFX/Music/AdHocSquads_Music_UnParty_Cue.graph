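# Sound Cue for the Ad Hoc Squads music sting, apparently played when a player leaves a party.
# A fixed half-second Delay node feeds a single Wave Player holding the Unjoin Party sound wave.
# The cue plays at reduced volume and is routed through the Fortnite events music sound class.
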SoundCue /AdHocSquads/Audio/SFX/Music/AdHocSquads_Music_UnParty_Cue
root=Delay_0
nodes [
 {id=Delay_0, children [WavePlayer_0], properties {delayMax=0.5, delayMin=0.5}}
 {id=WavePlayer_0, properties {soundWaveAssetPtr=&/AdHocSquads_Unjoin_Party}}
]
properties {maxDistance=2097152, soundClassObject=/Game/Sounds/FortSoundClassesAndModes/Fort_Events_Music, volumeMultiplier=0.6}
